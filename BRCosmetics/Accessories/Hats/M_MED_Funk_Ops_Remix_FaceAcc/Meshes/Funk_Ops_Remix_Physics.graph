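# Physics asset for the Funk Ops Remix cosmetic head accessory, defining bodies for its dynamic hair bones.
# Twelve hair bones carry small spheres, while the hair master bone and the head use capsules.
# Each hair bone constrains directly to the master in a star layout rather than chains, with angular limits and tighter twist on three strands.
# The head capsule has no constraint and approximates the skull volume.
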PhysicsAsset /BRCosmetics/Accessories/Hats/M_MED_Funk_Ops_Remix_FaceAcc/Meshes/Funk_Ops_Remix_Physics
bodies {
 dyn_hair_a2 [{shapeType=Sphere, radius=2, center=<0.3227 0 0>}]
 dyn_hair_b2 [{shapeType=Sphere, radius=2, center=<0.3227 0 0>}]
 dyn_hair_c2 [{shapeType=Sphere, radius=2, center=<0.3227 0 0>}]
 dyn_hair_d2 [{shapeType=Sphere, radius=2, center=<2.0417 -0.1434 0.1406>}]
 dyn_hair_e2 [{shapeType=Sphere, radius=1.75, center=<2.0417 -0.1434 0.1406>}]
 dyn_hair_f [{shapeType=Sphere, radius=2, center=<0.75 0 0>}]
 dyn_hair_g [{shapeType=Sphere, radius=2, center=<2.0417 -0.1434 0.1406>}]
 dyn_hair_h [{shapeType=Sphere, radius=2, center=<2.0417 -0.1434 0.1406>}]
 dyn_hair_i [{shapeType=Sphere, radius=2, center=<2.0417 -0.1434 0.1406>}]
 dyn_hair_j [{shapeType=Sphere, radius=2, center=<2.0417 -0.1434 0.1406>}]
 dyn_hair_k [{shapeType=Sphere, radius=2, center=<2.0417 -0.1434 0.1406>}]
 dyn_hair_l [{shapeType=Sphere, radius=1.969, center=<2.0417 -0.1434 0.1406>}]
 dyn_hair_mstr [{shapeType=Capsule, length=1, radius=1, rotation=<rot 90 0 0>}]
 head [{shapeType=Capsule, length=3, radius=5.75, center=<6.5722 2.0161 0>, rotation=<rot 90 0 90>}]
}
constraints [
 {bone1Name=dyn_hair_l, bone2Name=dyn_hair_mstr, swing1LimitDegrees=45, swing2LimitDegrees=45, twistLimitDegrees=45}
 {bone1Name=dyn_hair_k, bone2Name=dyn_hair_mstr, swing1LimitDegrees=45, swing2LimitDegrees=45, twistLimitDegrees=45}
 {bone1Name=dyn_hair_j, bone2Name=dyn_hair_mstr, swing1LimitDegrees=45, swing2LimitDegrees=45, twistLimitDegrees=45}
 {bone1Name=dyn_hair_i, bone2Name=dyn_hair_mstr, swing1LimitDegrees=45, swing2LimitDegrees=45, twistLimitDegrees=45}
 {bone1Name=dyn_hair_h, bone2Name=dyn_hair_mstr, swing1LimitDegrees=45, swing2LimitDegrees=45, twistLimitDegrees=45}
 {bone1Name=dyn_hair_g, bone2Name=dyn_hair_mstr, swing1LimitDegrees=45, swing2LimitDegrees=45, twistLimitDegrees=45}
 {bone1Name=dyn_hair_f, bone2Name=dyn_hair_mstr, swing1LimitDegrees=45, swing2LimitDegrees=45, twistLimitDegrees=45}
 {bone1Name=dyn_hair_d2, bone2Name=dyn_hair_mstr, swing1LimitDegrees=45, swing2LimitDegrees=45, twistLimitDegrees=45}
 {bone1Name=dyn_hair_e2, bone2Name=dyn_hair_mstr, swing1LimitDegrees=45, swing2LimitDegrees=45, twistLimitDegrees=45}
 {bone1Name=dyn_hair_b2, bone2Name=dyn_hair_mstr, swing1LimitDegrees=45, swing2LimitDegrees=45, twistLimitDegrees=15}
 {bone1Name=dyn_hair_a2, bone2Name=dyn_hair_mstr, swing1LimitDegrees=45, swing2LimitDegrees=45, twistLimitDegrees=15}
 {bone1Name=dyn_hair_c2, bone2Name=dyn_hair_mstr, swing1LimitDegrees=45, swing2LimitDegrees=45, twistLimitDegrees=15}
]
