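# Post-attach animation blueprint for the Funk Ops Remix face accessory, adding secondary physics motion on top of the pose copied from its attached parent mesh.
# A RigidBody node simulates the accessory's physics asset relative to spine_03 at half alpha, with damped component motion and overridden gravity.
# An AnimDynamics node and two rotation constraints for dyn_hair bones sit unconnected to the output, and the event graph holds no logic.
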AnimBlueprint /BRCosmetics/Accessories/Hats/M_MED_Funk_Ops_Remix_FaceAcc/Meshes/Funk_Ops_Remix_Skeleton_AnimBlueprint
target_skeleton=&/Funk_Ops_Remix_Skeleton
graphs [
 AnimGraph {nodes [
   Anim Dynamics - Bone: dyn_hair_a(Alpha=1, ExternalForce="0,0,0") {
    angularDampingOverride=1
    angularSpringConstant=1500
    bAngularSpring=true
    bGravityOverrideInSimSpace=true
    bOverrideAngularDamping=true
    gravityScale=0.35
    simulationSpace=RootRelative
    alphaBoolBlend {blendOption=Linear}
    boundBone {boneName=dyn_hair_a}
    componentAppliedLinearAccClamp=<1200 1200 1300>
    componentLinearAccScale=<1 1 0.6>
    componentLinearVelScale=<1 1 0.6>
    physicsBodyDefinitions [{
      sphereCollisionRadius=0
      boundBone {boneName=dyn_hair_a}
      boxExtents=<12 16 15>
      constraintSetup {angularLimitsMax=<0 40 40>, angularLimitsMin=<0 -40 -15>, angularTarget=<0 0 0>}
      localJointOffset=<0 0 0>
     }]
   }
   n1 = "Constraint - dyn_hair_e2 (R)"(Alpha=0, ConstraintWeights_0=1) {
    alphaBoolBlend {blendOption=Linear}
    alphaScaleBiasClamp {bInterpResult=true}
    boneToModify {boneName=dyn_hair_e2}
    constraintSetup [{offsetOption=None, transformType=Rotation, targetBone {boneName=dyn_hair_a2}}]
    constraintWeights [1]
   }
   "Constraint - dyn_hair_e (R)"(ComponentPose=@n1, Alpha=0, ConstraintWeights_0=1) {
    alphaBoolBlend {blendOption=Linear}
    alphaScaleBiasClamp {bInterpResult=true}
    boneToModify {boneName=dyn_hair_e}
    constraintSetup [{offsetOption=None, transformType=Rotation, targetBone {boneName=dyn_hair_a}}]
    constraintWeights [1]
   }
   n2 = RigidBody(ComponentPose=Local To Component(LocalPose=Copy Pose From Mesh() {bUseAttachedParent=true}), Alpha=0.5) {
    bClampLinearTranslationLimitToRefPose=true
    bOverrideWorldGravity=true
    overridePhysicsAsset=&/Funk_Ops_Remix_Physics
    simulationSpace=BaseBoneSpace
    alphaBoolBlend {blendOption=Linear}
    baseBoneRef {boneName=spine_03}
    componentAppliedLinearAccClamp=<4000 4000 4000>
    componentLinearAccScale=<0.33 0.33 0.33>
    componentLinearVelScale=<0.33 0.33 0.33>
    overrideWorldGravity {z=-980}
   }
   Output Pose(Result=Component To Local(ComponentPose=@n2)) {layerGroup=None}
   AnimGraphNode_SubInput_0 = Input Pose()
  ]}
 EventGraph {nodes [
   TryGetPawnOwner()
   Event BlueprintUpdateAnimation()
  ]}
]
variables []
